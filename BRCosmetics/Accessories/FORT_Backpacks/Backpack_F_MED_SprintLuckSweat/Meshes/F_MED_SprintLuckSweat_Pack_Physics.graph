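# Physics asset for the SprintLuckSweat backpack cosmetic, covering the pack and an attached figure-like piece with a body, arms and legs.
# Constraints form a chain from pack through MSTR_move and move to body, which branches to body_2 and four limb bones.
# Capsules shape every bone, with three on the pack.
# The body is angularly locked to move, and body_2 has one swing axis locked.
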PhysicsAsset /BRCosmetics/Accessories/FORT_Backpacks/Backpack_F_MED_SprintLuckSweat/Meshes/F_MED_SprintLuckSweat_Pack_Physics
bodies {
 MSTR_move [{shapeType=Capsule, length=0.505, radius=0.505, shapeName=MSTR_move_capsule, rotation=<rot 0 -90 0>}]
 arm_r [{shapeType=Capsule, length=4.6054, radius=2.623, shapeName=arm_r_capsule, center=<4 -1.75 1.953>, rotation=<rot 64.8046 -179.9996 -179.9999>}]
 arm_r1 [{
   shapeType=Capsule
   length=3.4472
   radius=2.6016
   shapeName=arm_r1_capsule
   center=<4 -1.8668 -0.7031>
   rotation=<rot -59.0315 179.9991 175.1818>
  }]
 body [{shapeType=Capsule, length=5.6988, radius=3.3159, center=<1.86 -0.5389 -3.01>, rotation=<rot 0.213 -90.0004 -0.7048>}]
 body_2 [{shapeType=Capsule, length=4.1374, radius=3.9329, shapeName=body_2_capsule, center=<2.46 -0.53 0.72>, rotation=<rot -79.3373 -0.0032 0.0033>}]
 leg_r [{
   shapeType=Capsule
   length=3.459
   radius=2.3338
   shapeName=leg_r_capsule
   center=<4.5037 -0.4706 -0.2327>
   rotation=<rot 88.4311 -0.0208 -14.3974>
  }]
 leg_r1 [{shapeType=Capsule, length=3.3823, radius=2.3235, shapeName=leg_r1_capsule, center=<1.8 -0.3 0.18>, rotation=<rot -84.3911 0.0005 -0.0007>}]
 move [{shapeType=Capsule, length=10, radius=2, shapeName=move_capsule, center=<8.1007 0 0>, rotation=<rot 0 -90 -90>}]
 pack [
  {shapeType=Capsule, length=24.2254, radius=4.437, center=<1.1449 -8.4 -10.3333>, rotation=<rot 89.2092 -179.645 -179.645>}
  {shapeType=Capsule, length=19.5575, radius=4.3173, shapeName=pack_capsule, center=<10.0843 -8.4 -5.816>, rotation=<rot -50 0 0>}
  {shapeType=Capsule, length=11.1669, radius=3.3057, shapeName=pack_capsule, center=<-6.3931 -8.4 -8.7916>, rotation=<rot -80 179.9968 0.0033>}
 ]
}
constraints [
 {
  bone1Name=body
  bone2Name=move
  swing1LimitDegrees=45
  swing1Motion=Locked
  swing2LimitDegrees=45
  swing2Motion=Locked
  twistLimitDegrees=45
  twistMotion=Locked
 }
 {bone1Name=leg_r1, bone2Name=body, swing1LimitDegrees=15, swing2LimitDegrees=15, twistLimitDegrees=0}
 {bone1Name=leg_r, bone2Name=body, swing1LimitDegrees=15, swing2LimitDegrees=15, twistLimitDegrees=0}
 {bone1Name=body_2, bone2Name=body, swing1LimitDegrees=45, swing1Motion=Locked, swing2LimitDegrees=15, twistLimitDegrees=0}
 {bone1Name=arm_r1, bone2Name=body, swing1LimitDegrees=15, swing2LimitDegrees=15, twistLimitDegrees=0}
 {bone1Name=arm_r, bone2Name=body, swing1LimitDegrees=15, swing2LimitDegrees=15, twistLimitDegrees=0}
 {bone1Name=move, bone2Name=MSTR_move, swing1LimitDegrees=0, swing2LimitDegrees=4, twistLimitDegrees=0}
 {bone1Name=MSTR_move, bone2Name=pack, swing1LimitDegrees=0, swing2LimitDegrees=4, twistLimitDegrees=0}
]
